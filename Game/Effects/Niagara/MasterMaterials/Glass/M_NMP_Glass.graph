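# M_NMP_Glass is a translucent glass master material for Niagara mesh particles, using Default Lit shading with per-pixel surface translucency lighting.
# A parameterized diffuse texture carries an alpha mask, which is multiplied by a scalar mask opacity parameter.
# Scalar parameters expose roughness, metallic and specular, and a normal map parameter with a flat default shares the mask texture's UVs.
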Material /Game/Effects/Niagara/MasterMaterials/Glass/M_NMP_Glass
graph {
 properties {
  bUsedWithNiagaraMeshParticles=true
  blendMode=BLEND_Translucent
  pixelDepthOffsetMode=PDOM_Legacy
  translucencyLightingMode=TLM_SurfacePerPixelLighting
  usedShadingModels=Default Lit
 }
 nodes [
  n1 = TextureCoordinate()
  n2 = TextureSampleParameter2D(UVs=@n1) {group=Mask Texture, parameterName=Diffuse Texture with Alpha Mask, sortPriority=10, texture=&T_BlacksRGB}
  Multiply(A=@n2.A, B=ScalarParameter() {defaultValue=1, group=Mask Texture, parameterName=Alpha Channel Mask Opacity})
  ScalarParameter() {defaultValue=0.4, group=Mask Texture, parameterName=Roughness, sortPriority=4}
  ScalarParameter() {group=Mask Texture, parameterName=Metallic, sortPriority=2}
  ScalarParameter() {defaultValue=0.5, group=Mask Texture, parameterName=Specular, sortPriority=3}
  TextureSampleParameter2D(UVs=@n1) {parameterName=NormalMap, samplerType=SAMPLERTYPE_Normal, texture=&FlatNormalBlue}
 ]
}
